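# MF_SchlickApprox1 is a material function in the engine's Substrate folder that converts an index of refraction into normal-incidence reflectance, the base term of Schlick's Fresnel approximation.
# It divides one minus the scalar IOR input by one plus IOR, then multiplies the result by itself.
# The IOR input is optional, falling back to a typical glass-like preview value when unconnected.
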MaterialFunction /Engine/Functions/Substrate/MF_SchlickApprox1
graph {nodes [
  n1 = FunctionInput() {bUsePreviewValueAsDefault=true, inputName=IOR, inputType=FunctionInput_Scalar, previewValue=<1.5 0 0 1>}
  n2 = Divide(A=OneMinus(None=@n1), B=Add(A=@n1, B=Constant() {r=1}))
  FunctionOutput(None=Multiply(A=@n2, B=@n2))
 ]}
