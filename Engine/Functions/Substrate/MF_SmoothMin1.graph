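# MF_SmoothMin1 is a Substrate material function that returns a smooth minimum of scalar inputs A and B, blended over a distance k.
# It derives a normalized overlap h from how far the absolute difference falls inside k, then subtracts h cubed times k times about one sixth from the hard minimum, a cubic polynomial smin.
# An If node outputs the plain minimum when the distance is zero.
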MaterialFunction /Engine/Functions/Substrate/MF_SmoothMin1
graph {nodes [
  n1 = FunctionInput() {bUsePreviewValueAsDefault=true, inputName=A, inputType=FunctionInput_Scalar}
  n2 = FunctionInput() {bUsePreviewValueAsDefault=true, inputName=B, inputType=FunctionInput_Scalar, sortPriority=1, previewValue=<1 0 0 1>}
  n3 = Min(A=@n1, B=@n2)
  n4 = FunctionInput() {
   bUsePreviewValueAsDefault=true
   description="blend distance of the two inputs (called k)"
   inputName=distance
   inputType=FunctionInput_Scalar
   sortPriority=2
   previewValue=<0.1 0 0 1>
  }
  n5 = Subtract(A=@n4, B=Abs(None=Subtract(A=@n1, B=@n2)))
  n6 = Power(Base=Divide(A=Max(A=@n5) {constB=0}, B=@n4) {desc=h}) {constExponent=3}
  n7 = Subtract(A=@n3, B=Multiply(A=Multiply(A=@n6, B=@n4)) {constB=0.1667})
  FunctionOutput(None=If(A=@n4, "A > B"=@n7, "A == B"=@n3, "A < B"=@n7)) {outputName=SmoothMin}
 ]}
