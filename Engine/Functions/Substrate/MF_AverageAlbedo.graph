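# Substrate material function that estimates average albedo from a scalar IOR input, which defaults to 1.5.
# It evaluates a closed-form fitted curve with hardcoded coefficients instead of integrating or sampling a lookup table.
# A linear expression in IOR is divided by a quadratic one, and the logarithm of that ratio is the single output.
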MaterialFunction /Engine/Functions/Substrate/MF_AverageAlbedo
graph {nodes [
  n1 = FunctionInput() {bUsePreviewValueAsDefault=true, inputName=IOR, inputType=FunctionInput_Scalar, previewValue=<1.5 0 0 1>}
  n2 = Subtract(A=Multiply(A=Constant() {r=10893}, B=@n1), B=Constant() {r=1438.2})
  n3 = Constant() {r=10212}
  n4 = Multiply(A=@n1, B=@n1)
  n5 = Constant() {r=-774.4}
  n6 = Add(A=Add(A=Multiply(A=@n1, B=@n3), B=Multiply(A=@n4, B=@n5)))
  FunctionOutput(None=Logarithm!(None=Divide(A=@n2, B=@n6)))
 ]}
