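# Substrate material function for OpenPBR coat darkening, which adjusts a base albedo for a coat layer of a given IOR.
# It estimates a coefficient K by blending a Schlick approximation with an average-albedo term divided by squared IOR, using base roughness as alpha.
# The factor (1-K)/(1-albedo*K) is faded in from 1 by the Darkening input, then multiplies the base albedo.
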MaterialFunction /Engine/Functions/Substrate/MF_OpenPBR_CoatDarkening
graph {nodes [
  n1 = FunctionInput() {bUsePreviewValueAsDefault=true, inputName=BaseAlbedo, sortPriority=10, previewValue=<0.8 0.8 0.8 1>}
  n2 = FunctionInput() {bUsePreviewValueAsDefault=true, inputName=CoatIOR, inputType=FunctionInput_Scalar, previewValue=<1.6 0 0 1>}
  n3 = OneMinus(None=MaterialFunctionCall!(IOR=@n2) {materialFunction=&/MF_AverageAlbedo}.Result)
  n4 = Multiply(A=@n2, B=@n2)
  n5 = FunctionInput() {bUsePreviewValueAsDefault=true, inputName=BaseRoughness, inputType=FunctionInput_Scalar, sortPriority=15, previewValue=<0.3 0 0 1>}
  n6 = LinearInterpolate(
   A=MaterialFunctionCall!(IOR=@n2) {materialFunction=&/MF_SchlickApprox}.Result
   B=OneMinus(None=Divide(A=@n3, B=@n4))
   Alpha=@n5
  ) {desc=K}
  n7 = Divide(A=OneMinus(None=@n6), B=OneMinus(None=Multiply(A=@n1, B=@n6)))
  n8 = FunctionInput() {bUsePreviewValueAsDefault=true, inputName=Darkening, sortPriority=5, previewValue=<1 0 0 1>}
  FunctionOutput(None=Multiply(A=@n1, B=LinearInterpolate(B=@n7, Alpha=@n8) {constA=1, desc=account for darkening + weight}))
 ]}
