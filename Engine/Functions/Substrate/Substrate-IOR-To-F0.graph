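# Material function exposed to the library under Substrate Helpers that converts a dielectric index of refraction (IOR) into an F0 value.
# A scalar IOR input, falling back to a preview value of 1.5 when unconnected, feeds a Custom HLSL node.
# That node calls the shader helper DielectricIorToF0 and returns the single float as the F0 output.
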MaterialFunction /Engine/Functions/Substrate/Substrate-IOR-To-F0
graph {
 properties {
  bExposeToLibrary=true
  description=Convert a dieletric IOR into a F0 value.
  userExposedCaption=Substrate IOR-To-F0
  libraryCategoriesText [Substrate Helpers]
 }
 nodes [
  MaterialExpressionFunctionInput_0 = FunctionInput() {bUsePreviewValueAsDefault=true, inputName=IOR, inputType=FunctionInput_Scalar, previewValue=<1.5 0 0 1>}
  n1 = Custom!(IOR=@MaterialExpressionFunctionInput_0) {
   code="return DielectricIorToF0(IOR);"
   outputType=CMOT_Float1
   inputs [{inputName=IOR, input {expression=&self:MaterialExpressionFunctionInput_0}}]
  }
  FunctionOutput(None=@n1) {description=F0 value based on input IOR, outputName=F0}
 ]
}
