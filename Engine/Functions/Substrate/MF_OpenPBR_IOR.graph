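# Material function for Substrate that computes effective specular IOR and F0 following the OpenPBR spec.
# It adjusts specular IOR for thin film and coat, with an IOR ratio flip named for total internal reflection, blended by coat weight.
# Specular weight, clamped, scales the resulting F0, which converts back to IOR through a signed square-root epsilon.
# Outputs are that IOR and its F0 via Substrate-IOR-To-F0.
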MaterialFunction /Engine/Functions/Substrate/MF_OpenPBR_IOR
graph {
 properties {description="Modulate F0 factor according to OpenPBR spec (https://academysoftwarefoundation.github.io/OpenPBR/##mjx-eqn-effective_coat_ior)"}
 nodes [
  n1 = NamedRerouteUsage()
  n2 = Divide(A=Add(A=@n1), B=Subtract(B=@n1))
  FunctionOutput(None=@n2) {outputName=IOR}
  FunctionOutput(None=MaterialFunctionCall!(IOR=@n2) {materialFunction=&/Substrate-IOR-To-F0}.F0) {outputName=F0, sortPriority=1}
  n3 = FunctionInput() {inputName=thin_film_ior, inputType=FunctionInput_Scalar, sortPriority=4}
  n4 = Sign(None=FunctionInput() {inputName=thin_film_weight, inputType=FunctionInput_Scalar, sortPriority=5})
  NamedRerouteDeclaration(None=Divide(A=FunctionInput() {inputName=specular_ior, inputType=FunctionInput_Scalar}, B=LinearInterpolate(B=@n3, Alpha=@n4) {constA=1})) {name=specular_ior, nodeColor=<rgba 1 0.2344 0 1>}
  n5 = NamedRerouteUsage()
  n6 = NamedRerouteUsage()
  n7 = Divide(A=@n6, B=@n5)
  NamedRerouteDeclaration(None=If(A=NamedRerouteUsage(), B=NamedRerouteUsage(), "A > B"=Divide(A=@n5, B=@n6), "A == B"=@n7, "A < B"=@n7) {constB=1}) {name=IOR_TIR, nodeColor=<rgba 0 1 0.9531 1>}
  n8 = NamedRerouteDeclaration(None=LinearInterpolate(A=NamedRerouteUsage(), B=NamedRerouteUsage(), Alpha=FunctionInput() {inputName=coat_weight, inputType=FunctionInput_Scalar, sortPriority=3})) {name=eta_s, nodeColor=<rgba 1 0.6328 0 1>}
  n9 = FunctionInput() {inputName=specular_weight, inputType=FunctionInput_Scalar, sortPriority=1}
  n10 = Divide(B=NamedRerouteUsage())
  n11 = NamedRerouteDeclaration(None=Multiply(A=MaterialFunctionCall(IOR=@n8) {materialFunction=&Substrate-IOR-To-F0}.F0, B=Clamp(None=@n9, Max=@n10))) {name=ModulatedF0, nodeColor=<rgba 0.0078 1 0 1>}
  n12 = Subtract(A=NamedRerouteUsage())
  NamedRerouteDeclaration(None=Multiply(A=SquareRoot(None=@n11), B=Sign(None=@n12))) {name=epsilon, nodeColor=<rgba 0 1 0.7891 1>}
  NamedRerouteDeclaration(None=FunctionInput() {inputName=coat_ior, inputType=FunctionInput_Scalar, sortPriority=2}) {name=coat_ior, nodeColor=<rgba 0 0.0859 1 1>}
 ]
}
