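# Material function for Substrate that converts OpenPBR translucent-base transmission color, depth, scatter and weight into a mean free path.
# It derives extinction from log-color absorption and scatter over depth, shifting negative absorption by its minimum component (only red and blue are tested), then scales the reciprocal.
# Lerps from black apply transmission weight, and If nodes select special cases, including a SubstrateTransmittanceToMFP conversion on an equality branch.
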MaterialFunction /Engine/Functions/Substrate/MF_TransmittanceToMeanFreePath
graph {
 properties {
  bExposeToLibrary=true
  description="Compute extinction coefficient and convert to mean free path according to OpenPBR spec. https://academysoftwarefoundation.github.io/OpenPBR/##model/dielectricbase/translucentbase"
 }
 nodes [
  n1 = NamedRerouteUsage()
  n2 = LinearInterpolate(
   A=Constant3Vector()
   B=NamedRerouteUsage()
   Alpha=NamedRerouteDeclaration(None=FunctionInput() {bUsePreviewValueAsDefault=true, inputName=TransmissionWeight, inputType=FunctionInput_Scalar}) {name=TransmissionWeight, nodeColor=<rgba 0.125 1 0 1>}
  )
  FunctionOutput(None=If(A=NamedRerouteUsage(), "A > B"=@n1, "A == B"=SubstrateTransmittanceToMFP!(TransmittanceColor=@n2, Thickness).MFP, "A < B"=@n1)) {outputName=MeanFreePath}
  n3 = NamedRerouteUsage()
  n4 = FunctionInput() {bUsePreviewValueAsDefault=true, inputName=TransmissionDepth, inputType=FunctionInput_Scalar}
  n5 = NamedRerouteDeclaration(None=Divide(A=FunctionInput() {bUsePreviewValueAsDefault=true, inputName=TransmissionScatter, sortPriority=2}, B=@n4)) {name=Us, nodeColor=<rgba 0.1719 0 1 1>}
  n6 = Max(A=NamedRerouteDeclaration(None=Add(A=@n3, B=@n5)) {name=Extinction, nodeColor=<rgba 0 1 0.25 1>}) {constB=0}
  NamedRerouteDeclaration(None=Multiply(A=Divide(B=@n6)) {constB=0.01}) {name=MFP, nodeColor=<rgba 1 0.6797 0 1>}
  n7 = FunctionInput() {bUsePreviewValueAsDefault=true, inputName=TransmissionColor, sortPriority=1, previewValue=<1 1 1 1>}
  n8 = Logarithm!(None=NamedRerouteDeclaration(None=@n7) {name=TransmissionColor, nodeColor=<rgba 0 1 0.7422 1>})
  n9 = NamedRerouteDeclaration(None=@n4) {name=TransmissionDepth, nodeColor=<rgba 1 0.3516 0 1>}
  NamedRerouteDeclaration(None=Subtract(B=Divide(A=@n8, B=@n9)) {constA=0}) {name=Ut, nodeColor=<rgba 0 0.9062 1 1>}
  n10 = Subtract(A=NamedRerouteUsage(), B=NamedRerouteUsage())
  n11 = MaterialFunctionCall(Float3=@n10) {materialFunction=&BreakOutFloat3Components}
  n12 = Min(A=Min(A=@n11.R, B=@n11.R), B=@n11.B)
  NamedRerouteDeclaration(None=If(A=@n12, "A > B"=@n10, "A == B"=@n10, "A < B"=Subtract(A=@n10, B=@n12))) {name=Ua, nodeColor=<rgba 1 0 0.5625 1>}
  NamedRerouteDeclaration(None=LinearInterpolate(A=Constant3Vector(), B=NamedRerouteUsage(), Alpha=NamedRerouteUsage())) {name=MFPLerp, nodeColor=<rgba 0 1 0.6719 1>}
  n13 = NamedRerouteUsage()
  NamedRerouteDeclaration(None=If(A=NamedRerouteUsage(), B=Constant3Vector() {constant=<rgba 1 1 1 0>}, "A > B"=@n13, "A == B"=NamedRerouteUsage(), "A < B"=@n13)) {name=MFPFinal, nodeColor=<rgba 1 0.3516 0 1>}
 ]
}
