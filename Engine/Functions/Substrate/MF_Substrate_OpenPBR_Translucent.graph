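# Substrate material function that implements a translucent OpenPBR surface, outputting front material, opacity mask and refraction IOR.
# Named reroutes feed a parameter-blended mix of dielectric transmission and metal slabs, with thin-film and coat-darkening adjustments, beneath a weighted, vertically layered coat slab.
# If nodes switch fuzz between coat and base slabs, and a path-tracing switch selects refraction IOR.
# Subsurface, dispersion and thin-walled inputs are exposed but unconnected.
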MaterialFunction /Engine/Functions/Substrate/MF_Substrate_OpenPBR_Translucent
graph {
 properties {bExposeToLibrary=true, libraryCategoriesText [Substrate]}
 nodes [
  FunctionInput() {
   bUsePreviewValueAsDefault=true
   description="Linearly scales the amount of dispersion. (not yet supported)"
   inputName=transmission_dispersion_scale
   inputType=FunctionInput_Scalar
   sortPriority=21
  }
  FunctionInput() {
   bUsePreviewValueAsDefault=true
   description=The bulk at the bottom of the material structure which consists of a mix of metal and dielectric properties
   inputName=----| Base |----
   inputType=FunctionInput_Substrate
   sortPriority=-1
  }
  FunctionInput() {bUsePreviewValueAsDefault=true, inputName=----| Specular |----, inputType=FunctionInput_Substrate, sortPriority=5}
  FunctionInput() {
   bUsePreviewValueAsDefault=true
   description="Physical Abbe number of the dielectric medium, describing how much the dielectric index of refraction varies across wavelengths. (not yet supported)"
   inputName=transmission_dispersion_abbe_number
   inputType=FunctionInput_Scalar
   sortPriority=22
  }
  FunctionInput() {
   bUsePreviewValueAsDefault=true
   description=This is a volumetric parametrization specifying the properties of a homogeneous medium interior to the object
   inputName=----| Transmission |----
   inputType=FunctionInput_Substrate
   sortPriority=15
  }
  FunctionInput() {
   bUsePreviewValueAsDefault=true
   description="A dielectric coat applied above all other layers excepting the fuzz.  This layer will affect the specular of the coated layers."
   inputName=----| Coat |----
   inputType=FunctionInput_Substrate
   sortPriority=29
  }
  FunctionInput() {
   bUsePreviewValueAsDefault=true
   description=For simulating cloth or dust or other retroreflective materials
   inputName=----| Fuzz |----
   inputType=FunctionInput_Substrate
   sortPriority=36
  }
  FunctionInput() {
   bUsePreviewValueAsDefault=true
   description="Emission is applied on the base slab, so it is attenuated by fuzz and (if enabled) the coat layer"
   inputName=----| Emission |----
   inputType=FunctionInput_Substrate
   sortPriority=40
  }
  FunctionInput() {
   bUsePreviewValueAsDefault=true
   description=Rainbow-like iridescence effects occur due to interference when a thin film with thickness of similar scale to light wavelengths coats a material.
   inputName=----| Thin Film |----
   inputType=FunctionInput_Substrate
   sortPriority=45
  }
  FunctionInput() {
   bUsePreviewValueAsDefault=true
   description=Modifications to the shading properties of the model itself.
   inputName=----| Geometry |----
   inputType=FunctionInput_Substrate
   sortPriority=50
  }
  FunctionInput() {
   bUsePreviewValueAsDefault=true
   description="Mix weight between subsurface and diffuse slabs [ 0, 1 ]"
   inputName=subsurface_weight
   inputType=FunctionInput_Scalar
   sortPriority=24
  }
  FunctionInput() {
   bUsePreviewValueAsDefault=true
   description="Subsurface color, i.e. the multi-scatter albedo [0, 1]"
   inputName=subsurface_color
   sortPriority=25
   previewValue=<0.8 0.8 0.8 1>
  }
  FunctionInput() {
   bUsePreviewValueAsDefault=true
   description="Radius of scatter (in cm) [ 0, inf )"
   inputName=subsurface_radius
   inputType=FunctionInput_Scalar
   sortPriority=26
   previewValue=<1 0 0 1>
  }
  FunctionInput() {
   bUsePreviewValueAsDefault=true
   description="Normalized Color (range 0 ->1) by which the radius (range 0 -> inf) is multiplied"
   inputName=subsurface_radius_scale
   sortPriority=27
   previewValue=<1 0.5 0.25 1>
  }
  FunctionInput() {
   bUsePreviewValueAsDefault=true
   description="Subsurface anisotropy (range -1 -> 1)"
   inputName=subsurface_anisotropy
   inputType=FunctionInput_Scalar
   sortPriority=28
  }
  FunctionInput() {
   bUsePreviewValueAsDefault=true
   description="Subsurface not used in translucent, just here for consistency with the opaque OpenPBR"
   inputName=----| Subsurface |----
   inputType=FunctionInput_Substrate
   sortPriority=23
  }
  FunctionInput(Preview=StaticBool()) {
   bUsePreviewValueAsDefault=true
   description=If true the surface is double-sided and represents an infinitesimally thin shell. Suitable for extremely geometrically thin objects such as leaves or paper.
   inputName=geometry_thin_walled
   inputType=FunctionInput_StaticBool
   sortPriority=52
  }
  MaterialExpressionStrataTransmittanceToMFP_0 = SubstrateTransmittanceToMFP!(TransmittanceColor=NamedRerouteUsage(), Thickness=Constant() {r=5})
  n1 = NamedRerouteUsage()
  n2 = Constant() {r=0.5}
  n3 = Constant()
  n4 = Constant3Vector()
  MaterialExpressionStrataSlabBSDF_0 = SubstrateSlabBSDF!(
   Diffuse Albedo=Constant3Vector() {desc=no coat MS in openPBR}
   F0=NamedRerouteUsage()
   F90
   Roughness=NamedRerouteUsage()
   Anisotropy=NamedRerouteUsage()
   Normal=NamedRerouteUsage()
   Tangent=NamedRerouteUsage()
   SSS MFP=@MaterialExpressionStrataTransmittanceToMFP_0.MFP
   SSS MFP Scale
   SSS Phase Anisotropy
   Emissive Color
   Second Roughness
   Second Roughness Weight
   Fuzz Roughness=If(A=@n1, "A > B"=NamedRerouteUsage(), "A == B"=@n2, "A < B"=@n2)
   Fuzz Amount=If(A=@n1, "A > B"=NamedRerouteUsage(), "A == B"=@n3, "A < B"=@n3)
   Fuzz Color=If(A=@n1, "A > B"=NamedRerouteUsage(), "A == B"=@n4, "A < B"=@n4)
   "Glint Density (Disabled)"
   "Glint UVs (Disabled)"
  ) {subSurfaceType=MSS_SimpleVolume}
  MaterialExpressionStrataWeight_0 = SubstrateWeight!(A=@MaterialExpressionStrataSlabBSDF_0, Weight=NamedRerouteUsage())
  n5 = NamedRerouteUsage()
  n6 = NamedRerouteUsage()
  n7 = NamedRerouteUsage()
  n8 = NamedRerouteUsage()
  n9 = NamedRerouteUsage()
  n10 = NamedRerouteUsage()
  n11 = NamedRerouteUsage()
  MaterialExpressionStrataThinFilm_0 = SubstrateThinFilm!(Normal=NamedRerouteUsage(), F0=NamedRerouteUsage(), F90, Thickness=NamedRerouteUsage(), IOR=NamedRerouteUsage())
  n12 = NamedRerouteUsage()
  n13 = NamedRerouteUsage()
  n14 = Constant3Vector() {constant=<rgba 1 1 1 0>}
  n15 = NamedRerouteUsage()
  n16 = NamedRerouteUsage()
  n17 = NamedRerouteUsage()
  n18 = NamedRerouteUsage()
  MaterialExpressionStrataSlabBSDF_1 = SubstrateSlabBSDF(
   Diffuse Albedo=LinearInterpolate(
    A=MaterialFunctionCall!(CoatIOR=@n5, Darkening=@n6, BaseAlbedo=@n7, BaseRoughness=@n8) {materialFunction=&/MF_OpenPBR_CoatDarkening}.Result
    B=Multiply(A=@n9, B=@n10)
    Alpha=NamedRerouteUsage()
   )
   F0=LinearInterpolate(A=NamedRerouteUsage(), B=If(A=@n11, "A > B"=@MaterialExpressionStrataThinFilm_0.Specular Color, "A == B"=@n12, "A < B"=@n12), Alpha=@n13)
   F90=LinearInterpolate(
    A=Constant3Vector() {constant=<rgba 1 1 1 0>}
    B=If(A=@n11, "A > B"=@MaterialExpressionStrataThinFilm_0.Edge Specular Color, "A == B"=@n14, "A < B"=@n14)
    Alpha=@n13
   )
   Roughness=NamedRerouteUsage()
   Anisotropy=NamedRerouteUsage()
   Normal=NamedRerouteUsage()
   Tangent=NamedRerouteUsage()
   SSS MFP=MaterialFunctionCall!(
    TransmissionDepth=NamedRerouteUsage()
    TransmissionWeight=NamedRerouteUsage()
    TransmissionColor=NamedRerouteUsage()
    TransmissionScatter=NamedRerouteUsage()
   ) {materialFunction=&/MF_TransmittanceToMeanFreePath}.MeanFreePath
   SSS Phase Anisotropy=NamedRerouteUsage()
   Emissive Color=NamedRerouteUsage()
   Second Roughness=@n8
   Second Roughness Weight=Constant() {r=1}
   Fuzz Roughness=If(A=@n15, "A > B"=Constant() {r=0.5}, "A == B"=@n18, "A < B"=@n18)
   Fuzz Amount=If(A=@n15, "A > B"=Constant(), "A == B"=@n17, "A < B"=@n17)
   Fuzz Color=If(A=@n15, "A > B"=Constant3Vector(), "A == B"=@n16, "A < B"=@n16)
  ) {subSurfaceType=MSS_SimpleVolume}
  MaterialExpressionStrataMetalnessToDiffuseAlbedoF0_1 = SubstrateMetalnessToDiffuseAlbedoF0!(BaseColor=NamedRerouteUsage(), Metallic=NamedRerouteUsage(), Specular=NamedRerouteUsage())
  n19 = NamedRerouteUsage()
  n20 = NamedRerouteUsage()
  MaterialExpressionStrataThinFilm_1 = SubstrateThinFilm(
   Normal=NamedRerouteUsage()
   F0=@MaterialExpressionStrataMetalnessToDiffuseAlbedoF0_1.F0
   F90=NamedRerouteUsage()
   Thickness=NamedRerouteUsage()
   IOR=NamedRerouteUsage()
  )
  n21 = NamedRerouteUsage()
  n22 = NamedRerouteUsage()
  n23 = NamedRerouteUsage()
  n24 = NamedRerouteUsage()
  n25 = NamedRerouteUsage()
  n26 = NamedRerouteUsage()
  MaterialExpressionStrataSlabBSDF_2 = SubstrateSlabBSDF(
   Diffuse Albedo=MaterialFunctionCall(
    CoatIOR=NamedRerouteUsage()
    Darkening=NamedRerouteUsage()
    BaseAlbedo=@MaterialExpressionStrataMetalnessToDiffuseAlbedoF0_1.DiffuseAlbedo
    BaseRoughness=@n19
   ) {materialFunction=&MF_OpenPBR_CoatDarkening}.Result
   F0=LinearInterpolate(
    A=@MaterialExpressionStrataMetalnessToDiffuseAlbedoF0_1.F0
    B=If(
     A=@n20
     "A > B"=@MaterialExpressionStrataThinFilm_1.Specular Color
     "A == B"=@MaterialExpressionStrataMetalnessToDiffuseAlbedoF0_1.F0
     "A < B"=@MaterialExpressionStrataMetalnessToDiffuseAlbedoF0_1.F0
    )
    Alpha=@n21
   )
   F90=LinearInterpolate(A=@n22, B=If(A=@n20, "A > B"=@MaterialExpressionStrataThinFilm_1.Edge Specular Color, "A == B"=@n22, "A < B"=@n22), Alpha=@n21)
   Roughness=@n19
   Anisotropy=NamedRerouteUsage()
   Normal=NamedRerouteUsage()
   Tangent=NamedRerouteUsage()
   Emissive Color=NamedRerouteUsage()
   Fuzz Roughness=If(A=@n23, "A > B"=Constant() {r=0.5}, "A == B"=@n24, "A < B"=@n24)
   Fuzz Amount=If(A=@n23, "A > B"=Constant(), "A == B"=@n25, "A < B"=@n25)
   Fuzz Color=If(A=@n23, "A > B"=Constant3Vector(), "A == B"=@n26, "A < B"=@n26)
  ) {subSurfaceType=MSS_SimpleVolume}
  MaterialExpressionStrataHorizontalMixing_0 = SubstrateHorizontalMixing!(Background=@MaterialExpressionStrataSlabBSDF_1, Foreground=@MaterialExpressionStrataSlabBSDF_2, Mix=NamedRerouteUsage()) {bUseParameterBlending=true, desc=Is Metal?}
  MaterialExpressionStrataVerticalLayering_0 = SubstrateVerticalLayering!(
   Top=@MaterialExpressionStrataWeight_0
   Bottom=@MaterialExpressionStrataHorizontalMixing_0
   Top Thickness=@MaterialExpressionStrataTransmittanceToMFP_0.Thickness
  )
  n27 = NamedRerouteUsage()
  MaterialExpressionStrataWeight_1 = SubstrateWeight(A=@MaterialExpressionStrataVerticalLayering_0, Weight=@n27)
  FunctionOutput(None=@MaterialExpressionStrataWeight_1) {outputName=OpenPBR_FrontMaterial}
  FunctionOutput(None=@n27) {outputName=OpacityMask, sortPriority=1}
  FunctionOutput(None=PathTracingQualitySwitch!(Normal=Constant() {r=1}, PathTraced=NamedRerouteUsage())) {outputName="Refraction (IOR)", sortPriority=2}
  n28 = FunctionInput() {
   bUsePreviewValueAsDefault=true
   description="Coverage weight of coat slab [ 0, 1 ]"
   inputName=coat_weight
   inputType=FunctionInput_Scalar
   sortPriority=30
  }
  NamedRerouteDeclaration(None=Saturate(None=@n28)) {name=CoatWeight, nodeColor=<rgba 0.01 1 0.34 1>}
  n29 = FunctionInput() {
   bUsePreviewValueAsDefault=true
   description=Square of normal-incidence transmittance of coat
   inputName=coat_color
   sortPriority=31
   previewValue=<1 1 1 1>
  }
  NamedRerouteDeclaration(None=SquareRoot(None=@n29)) {name=CoatColor, nodeColor=<rgba 0.01 1 0.34 1>}
  MaterialExpressionVertexTangentWS_2 = VertexTangentWS()
  n30 = Custom!(SurfaceTangent=@MaterialExpressionVertexTangentWS_2) {
   code="{\r\n##if MATERIAL_TANGENTSPACENORMAL\r\n\tconst float3 T = float3(1,0,0);\r\n##else\r\n\tconst float3 T = normalize(SurfaceTangent);\r\n##endif\r\n\treturn T;\r\n}"
   description=Default Tangent
   inputs [{
     inputName=SurfaceTangent
     input {expression=&self:MaterialExpressionVertexTangentWS_2}
    }]
  }
  n31 = FunctionInput(Preview=@n30) {
   bUsePreviewValueAsDefault=true
   description=This parameter controls the direction of anisotropic stretching for the material
   inputName=geometry_tangent
   sortPriority=54
  }
  NamedRerouteDeclaration(None=@n31) {name=BaseTangent, nodeColor=<rgba 1 0.01 0.67 1>}
  n32 = FunctionInput() {
   bUsePreviewValueAsDefault=true
   description=Apparent divergence of the fuzz from the surface normal direction
   inputName=fuzz_roughness
   inputType=FunctionInput_Scalar
   sortPriority=39
   previewValue=<0.5 0 0 1>
  }
  NamedRerouteDeclaration(None=@n32) {name=FuzzRoughness, nodeColor=<rgba 0.01 0.9175 1 1>}
  n33 = FunctionInput() {
   bUsePreviewValueAsDefault=true
   description=Distance traveled inside the medium by white light before its color becomes exactly TransmissionColor
   inputName=transmission_depth
   inputType=FunctionInput_Scalar
   sortPriority=18
  }
  NamedRerouteDeclaration(None=@n33) {name=TransmissionDepth, nodeColor=<rgba 0.4225 1 0.01 1>}
  n34 = FunctionInput() {
   bUsePreviewValueAsDefault=true
   description="Physical units: cd/m^2 (nits)"
   inputName=emission_luminance
   inputType=FunctionInput_Scalar
   sortPriority=41
  }
  n35 = FunctionInput() {
   bUsePreviewValueAsDefault=true
   description="Normalized color (range 0 ->1) by which the luminance (range 0 -> inf) is multiplied"
   inputName=emission_color
   sortPriority=42
   previewValue=<1 1 1 1>
  }
  NamedRerouteDeclaration(None=Multiply(A=@n34, B=Normalize(VectorInput=@n35))) {name=EmissiveColor, nodeColor=<rgba 0.01 0.175 1 1>}
  n36 = FunctionInput() {
   bUsePreviewValueAsDefault=true
   description="Thickness of the film in nm (range 0 - 1000)."
   inputName=thin_film_thickness
   inputType=FunctionInput_Scalar
   sortPriority=47
   previewValue=<0.5 0 0 1>
  }
  NamedRerouteDeclaration(None=Multiply(A=@n36, B=Constant() {r=0.31})) {name=ThinFilmThickness, nodeColor=<rgba 0.5875 0.01 1 1>}
  n37 = FunctionInput() {
   bUsePreviewValueAsDefault=true
   description="IOR of the film (range 1- 3)"
   inputName=thin_film_ior
   inputType=FunctionInput_Scalar
   sortPriority=48
   previewValue=<1.4 0 0 1>
  }
  NamedRerouteDeclaration(None=@n37) {name=ThinFilmIOR, nodeColor=<rgba 0.5875 0.01 1 1>}
  n38 = FunctionInput() {
   bUsePreviewValueAsDefault=true
   description="The presence weight of the entire surface, in effect an alpha mask."
   inputName=geometry_opacity
   inputType=FunctionInput_Scalar
   sortPriority=51
   previewValue=<1 0 0 1>
  }
  NamedRerouteDeclaration(None=@n38) {name=GeometryOpacity, nodeColor=<rgba 1 0.01 0.67 1>}
  n39 = FunctionInput() {
   bUsePreviewValueAsDefault=true
   description="Scalar multiplier to Base Color [ 0, 1 ]"
   inputName=base_weight
   inputType=FunctionInput_Scalar
   previewValue=<1 0 0 1>
  }
  n40 = Saturate(None=@n39)
  NamedRerouteDeclaration(None=@n40) {name=BaseWeight, nodeColor=<rgba 1 0.0925 0.01 1>}
  n41 = FunctionInput() {
   bUsePreviewValueAsDefault=true
   description="Albedo, in the case of dielectrics or the F0 color in the case of metals."
   inputName=base_color
   sortPriority=1
   previewValue=<0.8 0.8 0.8 1>
  }
  NamedRerouteDeclaration(None=LinearInterpolate(A=Multiply(A=@n40, B=@n41), B=NamedRerouteUsage(), Alpha=NamedRerouteUsage())) {name=DiffuseAlbedo, nodeColor=<rgba 1 0.0925 0.01 1>}
  n42 = FunctionInput() {
   bUsePreviewValueAsDefault=true
   description=Roughness of the diffuse microfacets
   inputName=base_diffuse_roughness
   inputType=FunctionInput_Scalar
   sortPriority=2
  }
  NamedRerouteDeclaration(None=@n42) {name=DiffuseRoughness, nodeColor=<rgba 1 0.0925 0.01 1>}
  n43 = FunctionInput() {
   bUsePreviewValueAsDefault=true
   description="IOR of the clearcoat (range 1 -> 3)"
   inputName=coat_ior
   inputType=FunctionInput_Scalar
   sortPriority=34
   previewValue=<1.6 0 0 1>
  }
  NamedRerouteDeclaration(None=MaterialFunctionCall!(IOR=@n43) {materialFunction=&/Substrate-IOR-To-F0}.F0) {name=CoatF0, nodeColor=<rgba 1 0 0.7031 1>}
  n44 = FunctionInput() {
   bUsePreviewValueAsDefault=true
   description="Blend factor for the mixture of the metallic base with the dielectric-base [ 0, 1 ]"
   inputName=base_metalness
   inputType=FunctionInput_Scalar
   sortPriority=3
  }
  NamedRerouteDeclaration(None=Saturate(None=@n44)) {name=BaseMetallic, nodeColor=<rgba 1 0.0925 0.01 1>}
  n45 = FunctionInput() {
   bUsePreviewValueAsDefault=true
   description="Anisotropy of NDF for base specular BSDF (range 0 -> 1)"
   inputName=specular_roughness_anisotropy
   inputType=FunctionInput_Scalar
   sortPriority=11
  }
  NamedRerouteDeclaration(None=Clamp(None=@n45) {minDefault=-1}) {name=SpecAnisotropy, nodeColor=<rgba 1 0.835 0.01 1>}
  n46 = FunctionInput() {
   bUsePreviewValueAsDefault=true
   description=Tints either the dielectric fresnel factor or the reflection albedo depending on the setting of the metallic property
   inputName=specular_color
   sortPriority=7
   previewValue=<1 1 1 1>
  }
  n47 = FunctionInput() {
   bUsePreviewValueAsDefault=true
   description="IOR of the specular lobe (range 1 -> 3)"
   inputName=specular_ior
   inputType=FunctionInput_Scalar
   sortPriority=9
   previewValue=<1.5 0 0 1>
  }
  n48 = FunctionInput() {
   bUsePreviewValueAsDefault=true
   description="Scalar multiplier to specular color  [ 0, 1 ]"
   inputName=specular_weight
   inputType=FunctionInput_Scalar
   sortPriority=6
   previewValue=<1 0 0 1>
  }
  n49 = MaterialFunctionCall!(
   specular_ior=NamedRerouteDeclaration(None=@n47) {name=SpecularIOR, nodeColor=<rgba 0 0.625 1 1>}
   specular_weight=@n48
   coat_ior=NamedRerouteUsage()
   coat_weight=NamedRerouteUsage()
   thin_film_ior=NamedRerouteUsage()
   thin_film_weight=NamedRerouteUsage()
  ) {materialFunction=&/MF_OpenPBR_IOR}
  NamedRerouteDeclaration(None=Multiply(A=@n46, B=@n49.F0)) {name=SpecF0, nodeColor=<rgba 1 0.835 0.01 1>}
  n50 = FunctionInput() {
   bUsePreviewValueAsDefault=true
   description=Roughness of NDF of specular lobe
   inputName=specular_roughness
   inputType=FunctionInput_Scalar
   sortPriority=8
   previewValue=<0.3 0 0 1>
  }
  NamedRerouteDeclaration(None=Saturate(None=@n50)) {name=SpecRoughness, nodeColor=<rgba 1 0.835 0.01 1>}
  NamedRerouteDeclaration(None=@n43) {name=CoatIOR, nodeColor=<rgba 0.01 1 0.34 1>}
  NamedRerouteDeclaration(None=@n41) {name=BaseColor, nodeColor=<rgba 1 0.0925 0.01 1>}
  NamedRerouteDeclaration(None=@n46) {name=SpecF90, nodeColor=<rgba 1 0.835 0.01 1>}
  n51 = FunctionInput() {
   bUsePreviewValueAsDefault=true
   description="Transmission anisotropy (range -1 -> 1)"
   inputName=transmission_scatter_anisotropy
   inputType=FunctionInput_Scalar
   sortPriority=20
  }
  NamedRerouteDeclaration(None=Clamp(None=@n51) {minDefault=-1}) {name=TransmissionAniso, nodeColor=<rgba 0.4225 1 0.01 1>}
  n52 = FunctionInput() {bUsePreviewValueAsDefault=true, description=Approximate scattering of the interior medium, inputName=transmission_scatter, sortPriority=19}
  NamedRerouteDeclaration(None=Saturate(None=@n52)) {name=TransmissionScatter, nodeColor=<rgba 0.4225 1 0.01 1>}
  n53 = FunctionInput() {
   bUsePreviewValueAsDefault=true
   description="Anisotropy of NDF of coat BSDF [0 , 1]"
   inputName=coat_roughness_anisotropy
   inputType=FunctionInput_Scalar
   sortPriority=33
  }
  NamedRerouteDeclaration(None=Saturate(None=@n53)) {name=CoatAnisotropy, nodeColor=<rgba 0.01 1 0.34 1>}
  n54 = FunctionInput() {
   bUsePreviewValueAsDefault=true
   description="Mix weight between transmissive and diffuse slabs [ 0, 1 ]"
   inputName=transmission_weight
   inputType=FunctionInput_Scalar
   sortPriority=16
  }
  NamedRerouteDeclaration(None=Saturate(None=@n54)) {name=TransmissionWeight, nodeColor=<rgba 0.4225 1 0.01 1>}
  n55 = FunctionInput() {
   bUsePreviewValueAsDefault=true
   description="Transmission color, used to specify the extinction color of the medium [0, 1]"
   inputName=transmission_color
   sortPriority=17
   previewValue=<1 1 1 1>
  }
  NamedRerouteDeclaration(None=Saturate(None=@n55)) {name=TransmissionColor, nodeColor=<rgba 0.4225 1 0.01 1>}
  n56 = Constant()
  n57 = Constant() {r=0.08}
  NamedRerouteDeclaration(None=Saturate(None=InverseLinearInterpolate(A=@n56, B=@n57, Value=@n49.F0) {constB=0.08})) {name=LegacySpec, nodeColor=<rgba 1 0.835 0.01 1>}
  n58 = FunctionInput() {
   bUsePreviewValueAsDefault=true
   description=Coverage weight of the thin film
   inputName=thin_film_weight
   inputType=FunctionInput_Scalar
   sortPriority=46
  }
  NamedRerouteDeclaration(None=Saturate(None=@n58)) {name=ThinFilmWeight, nodeColor=<rgba 0.5875 0.01 1 1>}
  n59 = FunctionInput() {
   bUsePreviewValueAsDefault=true
   description=Modulates the physical coat darkening effect.
   inputName=coat_darkening
   inputType=FunctionInput_Scalar
   sortPriority=35
   previewValue=<1 0 0 1>
  }
  NamedRerouteDeclaration(None=Multiply(A=Saturate(None=@n59), B=NamedRerouteUsage())) {name=CoatDarkening, nodeColor=<rgba 0.01 1 0.34 1>}
  n60 = FunctionInput(Preview=@n30) {
   bUsePreviewValueAsDefault=true
   description=This parameter controls the direction of anisotropic stretching for the material
   inputName=geometry_coat_tangent
   sortPriority=56
  }
  NamedRerouteDeclaration(None=@n60) {name=CoatTangent, nodeColor=<rgba 1 0.01 0.67 1>}
  n61 = FunctionInput() {
   bUsePreviewValueAsDefault=true
   description="Coverage weight of fuzz lobe [ 0, 1 ]"
   inputName=fuzz_weight
   inputType=FunctionInput_Scalar
   sortPriority=37
  }
  NamedRerouteDeclaration(None=Saturate(None=@n61)) {name=FuzzWeight, nodeColor=<rgba 0.01 0.9175 1 1>}
  n62 = FunctionInput() {bUsePreviewValueAsDefault=true, description=Reflection albedo color of fuzz, inputName=fuzz_color, sortPriority=38, previewValue=<1 1 1 1>}
  NamedRerouteDeclaration(None=@n62) {name=FuzzColor, nodeColor=<rgba 0.01 0.9175 1 1>}
  n63 = FunctionInput() {
   bUsePreviewValueAsDefault=true
   description=Roughness of NDF of coat BSDF
   inputName=coat_roughness
   inputType=FunctionInput_Scalar
   sortPriority=32
  }
  NamedRerouteDeclaration(None=Saturate(None=@n63)) {name=CoatRoughness, nodeColor=<rgba 0.01 1 0.34 1>}
  MaterialExpressionVertexNormalWS_3 = VertexNormalWS()
  n64 = Custom!(SurfaceNormal=@MaterialExpressionVertexNormalWS_3) {
   code="{\r\n##if MATERIAL_TANGENTSPACENORMAL\r\n\tconst float3 N = float3(0,0,1);\r\n##else\r\n\tconst float3 N = SurfaceNormal;\r\n##endif\r\n\treturn N;\r\n}"
   description=Default Normal
   inputs [{inputName=SurfaceNormal, input {expression=&self:MaterialExpressionVertexNormalWS_3}}]
  }
  n65 = FunctionInput(Preview=@n64) {bUsePreviewValueAsDefault=true, description=Perturbation of the coating shading normal, inputName=geometry_coat_normal, sortPriority=55}
  NamedRerouteDeclaration(None=@n65) {name=CoatNormal, nodeColor=<rgba 1 0.01 0.67 1>}
  n66 = FunctionInput(Preview=@n64) {bUsePreviewValueAsDefault=true, description=Perturbation of the 'substrate' shading normal, inputName=geometry_normal, sortPriority=53}
  NamedRerouteDeclaration(None=@n66) {name=BaseNormal, nodeColor=<rgba 1 0.01 0.67 1>}
 ]
}
