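# Substrate material function, exposed as "Substrate Coated Layer", that builds a coated surface from two slabs.
# The bottom slab takes metalness-style surface inputs, while the coat slab uses a black base color with its own specular, roughness, normal and emissive.
# Coat Color converts through TransmittanceToMFP into the coat's mean free path and a thickness, scaled by the saturated Thickness factor.
# Vertical layering stacks coat over base, and Opacity weights the result.
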MaterialFunction /Engine/Functions/Substrate/SMF_Coated
graph {
 properties {bExposeToLibrary=true, userExposedCaption=Substrate Coated Layer, libraryCategoriesText [Substrate Building Blocks]}
 nodes [
  n1 = FunctionInput() {bUsePreviewValueAsDefault=true, inputName=Coat Specular, inputType=FunctionInput_Scalar, previewValue=<0.5 0.5 0.5 1>}
  MaterialExpressionStrataMetalnessToDiffuseAlbedoF0_1 = SubstrateMetalnessToDiffuseAlbedoF0!(BaseColor=Constant3Vector(), Metallic, Specular=@n1)
  n2 = FunctionInput() {bUsePreviewValueAsDefault=true, inputName=Coat Roughness, inputType=FunctionInput_Scalar, sortPriority=1, previewValue=<0.1 0.1 0.1 1>}
  n3 = FunctionInput(Preview=Transform(None=VertexNormalWS()) {transformSourceType=TRANSFORMSOURCE_World, transformType=TRANSFORM_Tangent}) {bUsePreviewValueAsDefault=true, inputName=Coat Normal, sortPriority=2, previewValue=<0.5 0.5 1 1>}
  n4 = FunctionInput() {
   bUsePreviewValueAsDefault=true
   description=Transmittance color for a direction perpendicular to the surface.
   inputName=Coat Color
   sortPriority=3
   previewValue=<0.5 0.5 0.5 1>
  }
  MaterialExpressionStrataTransmittanceToMFP_0 = SubstrateTransmittanceToMFP!(TransmittanceColor=@n4, Thickness)
  MaterialExpressionStrataSlabBSDF_1 = SubstrateSlabBSDF!(
   Diffuse Albedo=@MaterialExpressionStrataMetalnessToDiffuseAlbedoF0_1.DiffuseAlbedo
   F0=@MaterialExpressionStrataMetalnessToDiffuseAlbedoF0_1.DiffuseAlbedo
   F90
   Roughness=@n2
   Anisotropy
   Normal=@n3
   Tangent
   SSS MFP=@MaterialExpressionStrataTransmittanceToMFP_0.MFP
   SSS MFP Scale
   SSS Phase Anisotropy
   Emissive Color=FunctionInput() {bUsePreviewValueAsDefault=true, inputName=Coat Emissive Color, sortPriority=4}
   Second Roughness
   Second Roughness Weight
   Fuzz Roughness
   Fuzz Amount
   Fuzz Color
   "Glint Density (Disabled)"
   "Glint UVs (Disabled)"
  )
  n5 = FunctionInput() {bUsePreviewValueAsDefault=true, inputName=Base Color, sortPriority=25, previewValue=<0.5 0.5 0.5 1>}
  n6 = FunctionInput() {bUsePreviewValueAsDefault=true, inputName=Specular, inputType=FunctionInput_Scalar, sortPriority=27, previewValue=<0.5 0.5 0.5 1>}
  MaterialExpressionStrataMetalnessToDiffuseAlbedoF0_0 = SubstrateMetalnessToDiffuseAlbedoF0(
   BaseColor=@n5
   Metallic=FunctionInput() {bUsePreviewValueAsDefault=true, inputName=Metallic, inputType=FunctionInput_Scalar, sortPriority=26}
   Specular=@n6
  )
  n7 = FunctionInput() {bUsePreviewValueAsDefault=true, inputName=Roughness, inputType=FunctionInput_Scalar, sortPriority=28, previewValue=<0.5 0.5 0.5 1>}
  n8 = FunctionInput(Preview=Transform(None=VertexNormalWS()) {transformSourceType=TRANSFORMSOURCE_World, transformType=TRANSFORM_Tangent}) {bUsePreviewValueAsDefault=true, inputName=Normal, sortPriority=29, previewValue=<0.5 0.5 1 1>}
  MaterialExpressionStrataSlabBSDF_0 = SubstrateSlabBSDF(
   Diffuse Albedo=@MaterialExpressionStrataMetalnessToDiffuseAlbedoF0_0.DiffuseAlbedo
   F0=@MaterialExpressionStrataMetalnessToDiffuseAlbedoF0_0.DiffuseAlbedo
   Roughness=@n7
   Normal=@n8
   Emissive Color=FunctionInput() {bUsePreviewValueAsDefault=true, inputName=Emissive Color, sortPriority=30}
  )
  n9 = FunctionInput() {bUsePreviewValueAsDefault=true, inputName=Thickness factor, inputType=FunctionInput_Scalar, sortPriority=5, previewValue=<1 1 1 1>}
  MaterialExpressionStrataVerticalLayering_0 = SubstrateVerticalLayering!(
   Top=@MaterialExpressionStrataSlabBSDF_1
   Bottom=@MaterialExpressionStrataSlabBSDF_0
   Top Thickness=Multiply(A=@MaterialExpressionStrataTransmittanceToMFP_0.Thickness, B=Saturate(None=@n9))
  )
  n10 = FunctionInput() {
   bUsePreviewValueAsDefault=true
   description=Opacity can be used to alpha fade out a surface when using any of the translucent blending modes.
   inputName=Opacity
   inputType=FunctionInput_Scalar
   sortPriority=31
   previewValue=<1 0 0 0>
  }
  MaterialExpressionStrataWeight_0 = SubstrateWeight!(A=@MaterialExpressionStrataVerticalLayering_0, Weight=@n10)
  FunctionOutput(None=@MaterialExpressionStrataWeight_0)
 ]
}
