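# Substrate material function that converts a glint density value, taking a scalar InDensity and returning OutDensity.
# Positive inputs are divided by a fixed constant and saturated into the zero-to-one range.
# An If node compares the input against its default threshold, so zero or negative densities output one instead of the scaled value.
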MaterialFunction /Engine/Functions/Substrate/SMF_GlintConversion
graph {
 properties {description=Conversion}
 nodes [
  n1 = FunctionInput() {inputName=InDensity, inputType=FunctionInput_Scalar}
  n2 = Divide(A=@n1) {constB=21}
  n3 = Constant() {r=1}
  FunctionOutput(None=If(A=@n1, "A > B"=Saturate(None=@n2), "A == B"=@n3, "A < B"=@n3)) {outputName=OutDensity}
 ]
}
