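# Substrate building-block material function, exposed as "Substrate UE4 Default Shading", that offers legacy Base Color, Metallic, Specular and Roughness inputs.
# A metalness conversion node feeds a Slab BSDF, with its DiffuseAlbedo output wired to both Diffuse Albedo and F0.
# Roughness, Emissive Color and Normal, defaulting to the tangent-space vertex normal, also feed the slab.
# Opacity weights the slab before output.
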MaterialFunction /Engine/Functions/Substrate/SMF_UE4Legacy
graph {
 properties {bExposeToLibrary=true, userExposedCaption=Substrate UE4 Default Shading, libraryCategoriesText [Substrate Building Blocks]}
 nodes [
  n1 = FunctionInput() {
   bUsePreviewValueAsDefault=true
   description=Defines the overall color of the Material.
   inputName=Base Color
   previewValue=<0.18 0.18 0.18 1>
  }
  n2 = FunctionInput() {
   bUsePreviewValueAsDefault=true
   description="Controls how \"metal-like\" your surface looks like. 0 means dielectric, 1 means conductor."
   inputName=Metallic
   inputType=FunctionInput_Scalar
   sortPriority=1
  }
  n3 = FunctionInput() {
   bUsePreviewValueAsDefault=true
   description=Used to scale the current amount of specularity on non-metallic surfaces and is a value between 0 and 1.
   inputName=Specular
   inputType=FunctionInput_Scalar
   sortPriority=2
   previewValue=<0.5 0.5 0.5 1>
  }
  MaterialExpressionStrataMetalnessToDiffuseAlbedoF0_0 = SubstrateMetalnessToDiffuseAlbedoF0!(BaseColor=@n1, Metallic=@n2, Specular=@n3)
  n4 = FunctionInput() {
   bUsePreviewValueAsDefault=true
   description="Controls how rough the Material is. Roughness of 0 (smooth) is a mirror reflection and 1 (rough) is completely matte or diffuse."
   inputName=Roughness
   inputType=FunctionInput_Scalar
   sortPriority=3
   previewValue=<0.5 0.5 0.5 1>
  }
  n5 = VertexNormalWS()
  n6 = FunctionInput(Preview=Normalize(VectorInput=Transform(None=@n5) {transformSourceType=TRANSFORMSOURCE_World, transformType=TRANSFORM_Tangent})) {
   bUsePreviewValueAsDefault=true
   description="The surface normal. (default value assumes normal is specified in tangent space)"
   inputName=Normal
   sortPriority=4
   previewValue=<0.5 0.5 1 1>
  }
  n7 = FunctionInput() {bUsePreviewValueAsDefault=true, description=Emissive Color., inputName=Emissive Color, sortPriority=5, previewValue=<0 0 0 0>}
  MaterialExpressionStrataSlabBSDF_0 = SubstrateSlabBSDF!(
   Diffuse Albedo=@MaterialExpressionStrataMetalnessToDiffuseAlbedoF0_0.DiffuseAlbedo
   F0=@MaterialExpressionStrataMetalnessToDiffuseAlbedoF0_0.DiffuseAlbedo
   F90
   Roughness=@n4
   Anisotropy
   Normal=@n6
   Tangent
   SSS MFP
   SSS MFP Scale
   SSS Phase Anisotropy
   Emissive Color=@n7
   Second Roughness
   Second Roughness Weight
   Fuzz Roughness
   Fuzz Amount
   Fuzz Color
   "Glint Density (Disabled)"
   "Glint UVs (Disabled)"
  )
  n8 = FunctionInput() {
   bUsePreviewValueAsDefault=true
   description=Opacity can be used to alpha fade out a surface when using any of the translucent blending modes.
   inputName=Opacity
   inputType=FunctionInput_Scalar
   sortPriority=9
   previewValue=<1 0 0 0>
  }
  MaterialExpressionStrataWeight_0 = SubstrateWeight!(A=@MaterialExpressionStrataSlabBSDF_0, Weight=@n8)
  FunctionOutput(None=@MaterialExpressionStrataWeight_0)
 ]
}
