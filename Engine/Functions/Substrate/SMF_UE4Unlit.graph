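# Substrate material function exposed to the library as "Substrate UE4 Unlit Shading", a building block that gives an unlit material an Opacity input.
# It saturates the scalar Opacity, multiplies the emissive Color input by it, and uses its inverse as transmittance color.
# Both feed a Substrate Unlit BSDF, which is the function's output, so coverage is expressed through emission and transmittance.
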MaterialFunction /Engine/Functions/Substrate/SMF_UE4Unlit
graph {
 properties {
  bExposeToLibrary=true
  description=Unlit material with Opacity input.
  userExposedCaption=Substrate UE4 Unlit Shading
  libraryCategoriesText [Substrate Building Blocks]
 }
 nodes [
  n1 = FunctionInput() {
   bUsePreviewValueAsDefault=true
   description="Represent the coverage of the material. 0 means 100% see through, 1 means fully opaque."
   inputName=Opacity
   inputType=FunctionInput_Scalar
   sortPriority=1
   previewValue=<0.5 0 0 1>
  }
  n2 = Saturate(None=@n1)
  MaterialExpressionStrataUnlitBSDF_0 = SubstrateUnlitBSDF!(
   EmissiveColor=Multiply(A=FunctionInput() {bUsePreviewValueAsDefault=true, description=Emissive color added to the scene., inputName=Color}, B=@n2)
   TransmittanceColor=OneMinus(None=@n2)
   Normal
  )
  FunctionOutput(None=@MaterialExpressionStrataUnlitBSDF_0)
 ]
}
