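# Substrate material function that converts a conductor's index of refraction into an F0 reflectance value, exposed to the library under Substrate Helpers.
# It takes IOR (n) and Absorption (k) inputs, each using its preview value as default, with IOR previewing at 1.5.
# A Custom HLSL node passes both to the shader helper ConductorIorToF0 and outputs the result.
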MaterialFunction /Engine/Functions/Substrate/Substrate-ConductorIOR-To-F0
graph {
 properties {
  bExposeToLibrary=true
  description=Convert a conductor IOR into a F0 value.
  userExposedCaption="Substrate IOR-To-F0 (Conductor)"
  libraryCategoriesText [Substrate Helpers]
 }
 nodes [
  MaterialExpressionFunctionInput_0 = FunctionInput() {bUsePreviewValueAsDefault=true, description="Index of refraction (n)", inputName=IOR, previewValue=<1.5 1.5 1.5 1>}
  MaterialExpressionFunctionInput_1 = FunctionInput() {bUsePreviewValueAsDefault=true, description="Absorption coefficient (k).", inputName=Absorption, sortPriority=1}
  n1 = Custom!(n=@MaterialExpressionFunctionInput_0, k=@MaterialExpressionFunctionInput_1) {
   code="return ConductorIorToF0(n, k);"
   showCode=true
   inputs [
    {inputName=n, input {expression=&self:MaterialExpressionFunctionInput_0}}
    {inputName=k, input {expression=&self:MaterialExpressionFunctionInput_1}}
   ]
  }
  FunctionOutput(None=@n1)
 ]
}
