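# Substrate FlipFlop material function, a Substrate helper that blends a facing F0 color to a tangent F90 color by view angle.
# A Custom node raises one minus the saturated dot product of camera vector and world normal to an exponent, then lerps with it.
# Falloff maps piecewise to that exponent, defaulting to a Schlick-like fifth power.
# It outputs the specular color and the LerpFactor, using the normalized vertex normal as default WorldNormal.
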MaterialFunction /Engine/Functions/Substrate/Substrate-FlipFlop
graph {
 properties {bExposeToLibrary=true, userExposedCaption=Substrate FlipFlop, libraryCategoriesText [Substrate Helpers]}
 nodes [
  MaterialExpressionFunctionInput_0 = FunctionInput(Preview=Constant3Vector() {constant=<rgba 1 0 0 0>}) {bUsePreviewValueAsDefault=true, description=Specular color when the surface is facing the camera., inputName=F0}
  MaterialExpressionFunctionInput_1 = FunctionInput(Preview=Constant3Vector() {constant=<rgba 0 1 0 0>}) {bUsePreviewValueAsDefault=true, description=Specular color when the surface is tangent to the camera., inputName=F90, sortPriority=1}
  n1 = FunctionInput(Preview=Constant() {desc="Default to 0.722, leading to a power of 5 similar to Schlick", r=0.7222}) {
   bUsePreviewValueAsDefault=true
   description="Describes the rate at which both facing and tangent colors are lerped. The curve is a power with exponent in [1/10, 10] mapped from the Falloff in [0,1]. Default to 0.722, leading to a power of 5 similar to Schlick"
   inputName=Falloff
   inputType=FunctionInput_Scalar
   sortPriority=2
  }
  MaterialExpressionSaturate_1 = Saturate(None=@n1)
  n2 = Normalize(VectorInput=VertexNormalWS())
  MaterialExpressionSaturate_0 = Saturate(None=DotProduct(
    A=CameraVectorWS()
    B=FunctionInput(Preview=@n2) {bUsePreviewValueAsDefault=true, description=The surface world normal., inputName=WorldNormal, sortPriority=3}
   ))
  n3 = Custom!(
   FacingColor=@MaterialExpressionFunctionInput_0
   TangentColor=@MaterialExpressionFunctionInput_1
   Falloff=@MaterialExpressionSaturate_1
   NoV=@MaterialExpressionSaturate_0
  ) {
   code="const float Exponent = Falloff < 0.5 ?\r\n1.0/(1.0+18.0 * (0.5-Falloff)) :\r\n1.0+18.0* (Falloff-0.5);\r\nLerpFactor = pow(1-NoV, Exponent);\r\nreturn FacingColor + (TangentColor-FacingColor) * LerpFactor;"
   additionalOutputs [{outputName=LerpFactor, outputType=CMOT_Float1}]
   inputs [
    {inputName=FacingColor, input {expression=&self:MaterialExpressionFunctionInput_0}}
    {inputName=TangentColor, input {expression=&self:MaterialExpressionFunctionInput_1}}
    {inputName=Falloff, input {expression=&self:MaterialExpressionSaturate_1}}
    {inputName=NoV, input {expression=&self:MaterialExpressionSaturate_0}}
   ]
  }
  FunctionOutput(None=@n3.return) {description=The resulting specular color.}
  FunctionOutput(None=@n3.LerpFactor) {description=The linear interpolation factor used to interpolate facing and tangent colors., outputName=LerpFactor, sortPriority=1}
 ]
}
