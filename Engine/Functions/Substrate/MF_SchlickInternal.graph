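# Substrate material function that computes normal-incidence reflectance (the Schlick F0 term) at the interface between two media from their indices of refraction.
# It divides the difference of scalar inputs Out_IOR and In_IOR by their sum, then squares the result.
# Unconnected inputs default to an air-like outside index and a glass-like inside index.
# It outputs only this base reflectance, with no view-angle term.
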MaterialFunction /Engine/Functions/Substrate/MF_SchlickInternal
graph {nodes [
  n1 = FunctionInput() {bUsePreviewValueAsDefault=true, inputName=Out_IOR, inputType=FunctionInput_Scalar, sortPriority=1, previewValue=<1 0 0 1>}
  n2 = FunctionInput() {bUsePreviewValueAsDefault=true, inputName=In_IOR, inputType=FunctionInput_Scalar, previewValue=<1.5 0 0 1>}
  n3 = Divide(A=Subtract(A=@n1, B=@n2), B=Add(A=@n1, B=@n2))
  FunctionOutput(None=Multiply(A=@n3, B=@n3))
 ]}
