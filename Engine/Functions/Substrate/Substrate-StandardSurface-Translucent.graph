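# Substrate material function exposing a translucent Standard Surface with base, specular, transmission, coat, sheen, emission and thin-film inputs.
# Two parameter-blended slabs, both with optional thin-film, sit under a weighted clear-coat slab, with transmission inputs converted to mean free path.
# It outputs the material (labelled "Opaque"), geometry opacity, and refraction IOR, which is 1 outside path tracing.
# Subsurface, thin_walled and dispersion inputs are unconnected; subsurface exists for parity with the opaque version.
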MaterialFunction /Engine/Functions/Substrate/Substrate-StandardSurface-Translucent
graph {
 properties {
  bExposeToLibrary=true
  description=Translucent Standard Surface
  userExposedCaption=Substrate Standard Surface Translucent
  libraryCategoriesText [Substrate Building Blocks]
 }
 nodes [
  FunctionInput(Preview=StaticBool()) {
   bUsePreviewValueAsDefault=true
   description=If true the surface is double-sided and represents an infinitely thin shell. Suitable for thin objects such as tree leaves or paper.
   inputName=thin_walled
   inputType=FunctionInput_StaticBool
   sortPriority=39
  }
  FunctionInput(Preview=Constant3Vector() {constant=<rgba 1 1 1 0>}) {
   bUsePreviewValueAsDefault=true
   description=The mean free path. The distance which light can travel before being scattered inside the surface.
   inputName=subsurface_radius
   sortPriority=95
  }
  FunctionInput(Preview=Constant() {r=1}) {
   bUsePreviewValueAsDefault=true
   description=Scalar weight for the subsurface radius value.
   inputName=subsurface_scale
   inputType=FunctionInput_Scalar
   sortPriority=100
  }
  FunctionInput(Preview=Constant()) {
   bUsePreviewValueAsDefault=true
   description=The blend between diffuse reflection and subsurface scattering. A value of 1.0 indicates full subsurface scattering and a value 0 for diffuse reflection only.
   inputName=subsurface
   inputType=FunctionInput_Scalar
   sortPriority=85
  }
  FunctionInput(Preview=Constant3Vector() {constant=<rgba 1 1 1 0>}) {bUsePreviewValueAsDefault=true, description=The color of the subsurface scattering effect., inputName=subsurface_color, sortPriority=90}
  FunctionInput(Preview=Constant()) {
   bUsePreviewValueAsDefault=true
   description="The direction of subsurface scattering. 0 scatters light evenly, positive values scatter forward and negative values scatter backward."
   inputName=subsurface_anisotropy
   inputType=FunctionInput_Scalar
   sortPriority=103
  }
  FunctionInput() {inputName="----|  Sheen  |----", inputType=FunctionInput_StaticBool, sortPriority=145}
  FunctionInput() {inputName="----|  Emission  |----", inputType=FunctionInput_StaticBool, sortPriority=160}
  FunctionInput() {inputName="----|  Specular  |----", inputType=FunctionInput_StaticBool, sortPriority=20}
  FunctionInput() {
   description="Subsurface not used in translucent, just here for consistency with the opaque StandardSurface"
   inputName="----|  Subsurface  |----"
   inputType=FunctionInput_StaticBool
   sortPriority=80
  }
  FunctionInput() {inputName="----|  Thin-Film  |----", inputType=FunctionInput_StaticBool, sortPriority=170}
  FunctionInput() {inputName="----|  Geometry  |----", inputType=FunctionInput_StaticBool, sortPriority=180}
  FunctionInput() {inputName="----|  Base  |----", inputType=FunctionInput_StaticBool}
  FunctionInput(Preview=Constant()) {
   bUsePreviewValueAsDefault=true
   description="Dispersion amount, describing how much the index of refraction varies across wavelengths. (not supported)"
   inputName=transmission_dispersion
   inputType=FunctionInput_Scalar
   sortPriority=78
  }
  FunctionInput() {inputName="----|  Transmission  |----", inputType=FunctionInput_StaticBool, sortPriority=55}
  FunctionInput() {inputName="----|  Coat  |----", inputType=FunctionInput_StaticBool, sortPriority=105}
  FunctionOutput(None=PathTracingQualitySwitch!(Normal=Constant() {r=1}, PathTraced=NamedRerouteUsage())) {outputName="Refraction (IOR)", sortPriority=2}
  n1 = FunctionInput(Preview=Constant() {r=1.5}) {
   bUsePreviewValueAsDefault=true
   description=The index of refraction of the clear-coat layer.
   inputName=coat_IOR
   inputType=FunctionInput_Scalar
   sortPriority=125
  }
  n2 = MaterialFunctionCall!(IOR=@n1) {materialFunction=&/Substrate-IOR-To-F0}
  n3 = FunctionInput(Preview=Constant() {r=0.1}) {
   bUsePreviewValueAsDefault=true
   description="The roughness of the clear-coat reflections. The lower the value, the sharper the reflection."
   inputName=coat_roughness
   inputType=FunctionInput_Scalar
   sortPriority=120
  }
  n4 = FunctionInput(Preview=Constant()) {
   bUsePreviewValueAsDefault=true
   description="The amount of directional bias, or anisotropy, of the clear-coat layer."
   inputName=coat_anisotropy
   inputType=FunctionInput_Scalar
   sortPriority=130
  }
  MaterialExpressionVertexNormalWS_4 = VertexNormalWS()
  n5 = Custom!(SurfaceNormal=@MaterialExpressionVertexNormalWS_4) {
   code="{\r\n##if MATERIAL_TANGENTSPACENORMAL\r\n\tconst float3 N = float3(0,0,1);\r\n##else\r\n\tconst float3 N = SurfaceNormal;\r\n##endif\r\n\treturn N;\r\n}"
   description=GetMtlDefNormalFrame
   inputs [{
     inputName=SurfaceNormal
     input {expression=&self:MaterialExpressionVertexNormalWS_4}
    }]
  }
  n6 = FunctionInput(Preview=@n5) {bUsePreviewValueAsDefault=true, description=Input normal for clear-coat layer, inputName=coat_normal, sortPriority=140}
  n7 = FunctionInput(Preview=Constant()) {
   bUsePreviewValueAsDefault=true
   description=The rotation of the anisotropic effect of the clear-coat layer.
   inputName=coat_rotation
   inputType=FunctionInput_Scalar
   sortPriority=135
  }
  MaterialExpressionFrac_1 = Frac(None=@n7)
  MaterialExpressionReroute_18 = Reroute(None=@n6)
  MaterialExpressionNamedRerouteUsage_77 = NamedRerouteUsage()
  n8 = Custom!(Rotation=@MaterialExpressionFrac_1, SurfaceNormal=@MaterialExpressionReroute_18, SurfaceTangent=@MaterialExpressionNamedRerouteUsage_77) {
   code="{\r\n\tconst float AngleRad = saturate(Rotation) * PI * 2;\r\n\tconst float3 N = SurfaceNormal;\r\n\tconst float3 T = normalize(SurfaceTangent);\r\n\tconst float3 B = normalize(cross(SurfaceTangent, SurfaceNormal));\r\n\tconst float3 Out = cos(AngleRad) * T + sin(AngleRad) * B;\r\n\treturn Out;\r\n}"
   description=RotateTangentVec
   inputs [
    {inputName=Rotation, input {expression=&self:MaterialExpressionFrac_1}}
    {inputName=SurfaceNormal, input {expression=&self:MaterialExpressionReroute_18}}
    {
     inputName=SurfaceTangent
     input {expression=&self:MaterialExpressionNamedRerouteUsage_77}
    }
   ]
  }
  n9 = FunctionInput(Preview=Constant3Vector() {constant=<rgba 1 1 1 0>}) {bUsePreviewValueAsDefault=true, description=The color of the clear-coat layer's transparency., inputName=coat_color, sortPriority=115}
  n10 = SubstrateTransmittanceToMFP!(TransmittanceColor=SquareRoot(None=@n9), Thickness=Constant() {r=5})
  n11 = FunctionInput(Preview=Constant()) {
   bUsePreviewValueAsDefault=true
   description=The weight of a reflective clear-coat layer on top of the material. Use for materials such as car paint or an oily layer.
   inputName=coat
   inputType=FunctionInput_Scalar
   sortPriority=110
  }
  n12 = NamedRerouteDeclaration(None=@n11) {name=CoatWeight, nodeColor=<rgba 1 0 0.9609 1>}
  n13 = SubstrateWeight!(
   A=SubstrateSlabBSDF!(
    Diffuse Albedo=Constant3Vector()
    F0=Saturate(None=@n2.F0)
    F90=Constant3Vector() {constant=<rgba 1 1 1 0>}
    Roughness=NamedRerouteDeclaration(None=@n3) {name=CoatRoughness, nodeColor=<rgba 0 1 0.7188 1>}
    Anisotropy=@n4
    Normal=@n6
    Tangent=@n8
    SSS MFP=@n10.MFP
    SSS MFP Scale
    SSS Phase Anisotropy
    Emissive Color
    Second Roughness
    Second Roughness Weight
    Fuzz Roughness
    Fuzz Amount
    Fuzz Color
    "Glint Density (Disabled)"
    "Glint UVs (Disabled)"
   ) {subSurfaceType=MSS_SimpleVolume}
   Weight=@n12
  )
  n14 = NamedRerouteUsage()
  n15 = NamedRerouteUsage()
  n16 = NamedRerouteUsage()
  n17 = NamedRerouteUsage()
  n18 = SubstrateThinFilm!(Normal=@n14, F0=@n15, F90=@n16, Thickness=@n17, IOR=NamedRerouteUsage())
  n19 = Power(Base=NamedRerouteUsage(), Exp=NamedRerouteUsage())
  n20 = NamedRerouteUsage()
  n21 = NamedRerouteUsage()
  n22 = NamedRerouteUsage()
  n23 = SubstrateMetalnessToDiffuseAlbedoF0!(BaseColor=NamedRerouteUsage(), Metallic=NamedRerouteUsage(), Specular=NamedRerouteUsage())
  n24 = NamedRerouteUsage()
  n25 = SubstrateThinFilm(Normal=@n22, F0=@n23.F0, F90=NamedRerouteUsage(), Thickness=@n24, IOR=NamedRerouteUsage())
  n26 = NamedRerouteUsage()
  n27 = SubstrateHorizontalMixing!(
   Background=SubstrateSlabBSDF(
    Diffuse Albedo=LinearInterpolate(A=@n19, B=@n20, Alpha=@n21)
    F0=If(A=@n17, "A > B"=@n18.Specular Color, "A == B"=@n15, "A < B"=@n15)
    F90=If(A=@n17, "A > B"=@n18.Edge Specular Color, "A == B"=@n16, "A < B"=@n16)
    Roughness=NamedRerouteUsage()
    Anisotropy=NamedRerouteUsage()
    Normal=@n14
    Tangent=NamedRerouteUsage()
    SSS MFP=NamedRerouteUsage()
    Emissive Color=NamedRerouteUsage()
    Second Roughness=NamedRerouteUsage()
    Second Roughness Weight=Constant() {r=1}
    Fuzz Roughness=NamedRerouteUsage()
    Fuzz Amount=NamedRerouteUsage()
    Fuzz Color=NamedRerouteUsage()
   ) {subSurfaceType=MSS_SimpleVolume}
   Foreground=SubstrateSlabBSDF(
    Diffuse Albedo=@n23.DiffuseAlbedo
    F0=If(A=@n24, "A > B"=@n25.Specular Color, "A == B"=@n23.F0, "A < B"=@n23.F0)
    F90=If(A=@n24, "A > B"=@n25.Edge Specular Color, "A == B"=@n26, "A < B"=@n26)
    Roughness=NamedRerouteUsage()
    Anisotropy=NamedRerouteUsage()
    Normal=@n22
    Tangent=NamedRerouteUsage()
    Emissive Color=NamedRerouteUsage()
   ) {subSurfaceType=MSS_SimpleVolume}
   Mix=NamedRerouteUsage()
  ) {bUseParameterBlending=true}
  n28 = NamedRerouteUsage()
  FunctionOutput(None=SubstrateWeight(A=SubstrateVerticalLayering!(Top=@n13, Bottom=@n27, Top Thickness=@n10.Thickness) {desc=Coat}, Weight=@n28)) {outputName=Substrate StandardSurface Opaque}
  FunctionOutput(None=@n28) {outputName=Geometry Opacity, sortPriority=1}
  MaterialExpressionVertexTangentWS_2 = VertexTangentWS()
  n29 = Custom!(SurfaceTangent=@MaterialExpressionVertexTangentWS_2) {
   code="{\r\n##if MATERIAL_TANGENTSPACENORMAL\r\n\tconst float3 T = float3(1,0,0);\r\n##else\r\n\tconst float3 T = normalize(SurfaceTangent);\r\n##endif\r\n\treturn T;\r\n}"
   description=GetMtlDefTangentFrame
   inputs [{
     inputName=SurfaceTangent
     input {expression=&self:MaterialExpressionVertexTangentWS_2}
    }]
  }
  NamedRerouteDeclaration(None=FunctionInput(Preview=@n29) {bUsePreviewValueAsDefault=true, description=Input geometric tangent, inputName=tangent, sortPriority=195}) {name=AnisotropyTangent, nodeColor=<rgba 0 0.0625 1 1>}
  n30 = Constant3Vector() {constant=<rgba 1 1 1 0>}
  n31 = FunctionInput(Preview=Constant()) {
   bUsePreviewValueAsDefault=true
   description=The amount of emitted incandescent light.
   inputName=emission
   inputType=FunctionInput_Scalar
   sortPriority=164
  }
  NamedRerouteDeclaration(None=Multiply(
    A=FunctionInput(Preview=@n30) {bUsePreviewValueAsDefault=true, description=The color of the emitted light., inputName=emission_color, sortPriority=165}
    B=@n31
   )) {name=EmissiveColor, nodeColor=<rgba 0 1 0.7656 1>}
  n32 = FunctionInput(Preview=Constant() {r=1}) {
   bUsePreviewValueAsDefault=true
   description=Multiplier on the intensity of the diffuse reflection.
   inputName=base
   inputType=FunctionInput_Scalar
   sortPriority=2
  }
  NamedRerouteDeclaration(None=@n32) {name=BaseWeight, nodeColor=<rgba 0 1 0.6719 1>}
  n33 = FunctionInput(Preview=Constant()) {
   bUsePreviewValueAsDefault=true
   description=The weight of a sheen layer that can be used to approximate microfibers or fabrics such as velvet and satin.
   inputName=sheen
   inputType=FunctionInput_Scalar
   sortPriority=150
  }
  NamedRerouteDeclaration(None=@n33) {name=SheenWeight, nodeColor=<rgba 0.8984 0 1 1>}
  NamedRerouteDeclaration(None=FunctionInput(Preview=Constant3Vector() {constant=<rgba 1 1 1 0>}) {bUsePreviewValueAsDefault=true, description=The color of the sheen layer., inputName=sheen_color, sortPriority=155}) {name=SheenColor, nodeColor=<rgba 1 0 0.7266 1>}
  n34 = FunctionInput(Preview=Constant()) {
   bUsePreviewValueAsDefault=true
   description=Controls the saturation of diffuse reflection and subsurface scattering below the clear-coat.
   inputName=coat_affect_color
   inputType=FunctionInput_Scalar
   sortPriority=140
  }
  NamedRerouteDeclaration(None=Add(A=Multiply(A=@n12, B=@n34))) {name=CoatAffectColor, nodeColor=<rgba 1 0.3281 0 1>}
  n35 = FunctionInput(Preview=Constant() {r=1.5}) {
   bUsePreviewValueAsDefault=true
   description=Index of refraction for specular reflection.
   inputName=specular_IOR
   inputType=FunctionInput_Scalar
   sortPriority=40
  }
  n36 = MaterialFunctionCall!(In_IOR=@n35, Out_IOR=Constant() {r=1}) {materialFunction=&/MF_SchlickInternal}
  n37 = FunctionInput(Preview=Constant3Vector() {constant=<rgba 1 1 1 0>}) {bUsePreviewValueAsDefault=true, description=Color tint on the specular reflection., inputName=specular_color, sortPriority=30}
  n38 = FunctionInput(Preview=Constant() {r=1}) {
   bUsePreviewValueAsDefault=true
   description=Multiplier on the intensity of the specular reflection.
   inputName=specular
   inputType=FunctionInput_Scalar
   sortPriority=25
  }
  n39 = Multiply(A=@n37, B=NamedRerouteDeclaration(None=@n38) {name=SpecularIOR, nodeColor=<rgba 0 0.4844 1 1>})
  NamedRerouteDeclaration(None=Multiply(A=@n36.Result, B=@n39)) {name=SpecF0, nodeColor=<rgba 1 0.5625 0 1>}
  MaterialExpressionVertexNormalWS_2 = VertexNormalWS()
  n40 = Custom!(SurfaceNormal=@MaterialExpressionVertexNormalWS_2) {
   code="{\r\n##if MATERIAL_TANGENTSPACENORMAL\r\n\tconst float3 N = float3(0,0,1);\r\n##else\r\n\tconst float3 N = SurfaceNormal;\r\n##endif\r\n\treturn N;\r\n}"
   description=GetMtlDefNormalFrame
   inputs [{
     inputName=SurfaceNormal
     input {expression=&self:MaterialExpressionVertexNormalWS_2}
    }]
  }
  NamedRerouteDeclaration(None=FunctionInput(Preview=@n40) {bUsePreviewValueAsDefault=true, description=Input geometric normal, inputName=normal, sortPriority=190}) {name=Normal, nodeColor=<rgba 0.8281 1 0 1>}
  n41 = FunctionInput(Preview=Constant()) {
   bUsePreviewValueAsDefault=true
   description=Roughness of the diffuse reflection. Higher values cause the surface to appear flatter and darker.
   inputName=diffuse_roughness
   inputType=FunctionInput_Scalar
   sortPriority=10
  }
  n42 = FunctionInput(Preview=Constant()) {
   bUsePreviewValueAsDefault=true
   description="Specifies how metallic the material appears. At its maximum, the surface behaves like a metal, using fully specular reflection."
   inputName=metalness
   inputType=FunctionInput_Scalar
   sortPriority=15
  }
  NamedRerouteDeclaration(None=LinearInterpolate(A=@n41, Alpha=@n42) {constB=0}) {name=DiffuseRoughness, nodeColor=<rgba 0 1 0.6719 1>}
  n43 = FunctionInput(Preview=Constant() {r=0.3}) {
   bUsePreviewValueAsDefault=true
   description=The roughness of the sheen layer.
   inputName=sheen_roughness
   inputType=FunctionInput_Scalar
   sortPriority=157
  }
  NamedRerouteDeclaration(None=@n43) {name=SheenRoughness, nodeColor=<rgba 1 0.7031 0 1>}
  n44 = FunctionInput(Preview=Constant() {r=1}) {
   bUsePreviewValueAsDefault=true
   description=The opacity of the entire material.
   inputName=opacity
   inputType=FunctionInput_Scalar
   sortPriority=187
  }
  NamedRerouteDeclaration(None=@n44) {name=GeometryOpacity, nodeColor=<rgba 1 0.8672 0 1>}
  n45 = Constant3Vector() {constant=<rgba 0.8 0.8 0.8 0>}
  NamedRerouteDeclaration(None=Multiply(
    A=@n32
    B=FunctionInput(Preview=@n45) {bUsePreviewValueAsDefault=true, description=Color of the diffuse reflection., inputName=base_color, sortPriority=5}
   )) {name=WeightedBase, nodeColor=<rgba 0.0781 1 0 1>}
  n46 = FunctionInput(Preview=Constant()) {
   bUsePreviewValueAsDefault=true
   description=Specifies the distance light travels inside the material before its becomes exactly the transmission color
   inputName=transmission_depth
   inputType=FunctionInput_Scalar
   sortPriority=70
  }
  n47 = FunctionInput(Preview=Constant()) {
   bUsePreviewValueAsDefault=true
   description=Transmission of light through the surface for materials such as glass or water. The greater the value the more transparent the material.
   inputName=transmission
   inputType=FunctionInput_Scalar
   sortPriority=60
  }
  n48 = Saturate(None=@n47)
  n49 = FunctionInput(Preview=Constant3Vector() {constant=<rgba 1 1 1 0>}) {bUsePreviewValueAsDefault=true, description=Color tint on the transmitted light., inputName=transmission_color, sortPriority=65}
  n50 = FunctionInput(Preview=Constant3Vector()) {bUsePreviewValueAsDefault=true, inputName=transmission_scatter, sortPriority=75}
  NamedRerouteDeclaration(None=MaterialFunctionCall!(TransmissionDepth=@n46, TransmissionWeight=@n48, TransmissionColor=@n49, TransmissionScatter=@n50) {materialFunction=&/MF_TransmittanceToMeanFreePath}.MeanFreePath) {name=MFP, nodeColor=<rgba 1 0 0.1875 1>}
  n51 = Constant()
  n52 = Constant() {r=0.08}
  NamedRerouteDeclaration(None=Saturate(None=InverseLinearInterpolate(A=@n51, B=@n52, Value=@n36.Result))) {name=SlabIOR, nodeColor=<rgba 0 1 1 1>}
  n53 = FunctionInput(Preview=Constant()) {
   bUsePreviewValueAsDefault=true
   description="Additional roughness on top of specular roughness. Positive values blur refractions more than reflections, and negative values blur refractions less."
   inputName=transmission_extra_roughness
   inputType=FunctionInput_Scalar
   sortPriority=77
  }
  NamedRerouteDeclaration(None=Add(A=@n53, B=NamedRerouteUsage())) {name=TransmissionRough, nodeColor=<rgba 0.3125 0 1 1>}
  NamedRerouteDeclaration(None=@n48) {name=TransmissionWeight, nodeColor=<rgba 0 0.9766 1 1>}
  NamedRerouteDeclaration(None=@n50) {name=TransmissionScatter, nodeColor=<rgba 0 0.9531 1 1>}
  n54 = FunctionInput(Preview=Constant()) {
   bUsePreviewValueAsDefault=true
   description=The thickness of the thin film layer on a surface. Use for materials such as multitone car paint or soap bubbles.
   inputName=thin_film_thickness
   inputType=FunctionInput_Scalar
   sortPriority=175
  }
  NamedRerouteDeclaration(None=Multiply(A=@n54) {constB=0.0003}) {name=FilmThickness, nodeColor=<rgba 0 0.6016 1 1>}
  n55 = FunctionInput(Preview=Constant() {r=1.5}) {
   bUsePreviewValueAsDefault=true
   description=The index of refraction of the medium surrounding the material.
   inputName=thin_film_IOR
   inputType=FunctionInput_Scalar
   sortPriority=177
  }
  NamedRerouteDeclaration(None=@n55) {name=FilmIOR, nodeColor=<rgba 0 1 0.0625 1>}
  NamedRerouteDeclaration(None=@n42) {name=Metalness, nodeColor=<rgba 1 0.0234 0 1>}
  NamedRerouteDeclaration(None=@n39) {name=WeightedSpec, nodeColor=<rgba 1 0.9609 0 1>}
  n56 = FunctionInput(Preview=Constant() {r=0.2}) {
   bUsePreviewValueAsDefault=true
   description="The roughness of the specular reflection. Lower numbers produce sharper reflections, higher numbers produce blurrier reflections."
   inputName=specular_roughness
   inputType=FunctionInput_Scalar
   sortPriority=35
  }
  n57 = FunctionInput(Preview=Constant()) {
   bUsePreviewValueAsDefault=true
   description=Controls the roughness of the specular reflection in the layers below the clear-coat.
   inputName=coat_affect_roughness
   inputType=FunctionInput_Scalar
   sortPriority=141
  }
  n58 = Multiply(A=@n57, B=NamedRerouteUsage())
  n59 = NamedRerouteUsage()
  NamedRerouteDeclaration(None=LinearInterpolate(A=@n56, Alpha=Multiply(A=@n58, B=@n59))) {name=SpecRoughness, nodeColor=<rgba 0.7812 0 1 1>}
  n60 = FunctionInput(Preview=Constant()) {
   bUsePreviewValueAsDefault=true
   description=The directional bias of reflected and transmitted light resulting in materials appearing rougher or glossier in certain directions.
   inputName=specular_anisotropy
   inputType=FunctionInput_Scalar
   sortPriority=45
  }
  NamedRerouteDeclaration(None=@n60) {name=SpecAnisotropy, nodeColor=<rgba 0 1 0.4141 1>}
  n61 = FunctionInput(Preview=Constant()) {
   bUsePreviewValueAsDefault=true
   description=Rotation of the axis of specular anisotropy around the surface normal.
   inputName=specular_rotation
   inputType=FunctionInput_Scalar
   sortPriority=50
  }
  MaterialExpressionFrac_0 = Frac(None=@n61)
  MaterialExpressionNamedRerouteUsage_76 = NamedRerouteUsage()
  MaterialExpressionNamedRerouteUsage_0 = NamedRerouteUsage()
  n62 = Custom!(
   Rotation=@MaterialExpressionFrac_0
   SurfaceNormal=@MaterialExpressionNamedRerouteUsage_76
   SurfaceTangent=@MaterialExpressionNamedRerouteUsage_0
  ) {
   code="{\r\n\tconst float AngleRad = saturate(Rotation) * PI * 2;\r\n\tconst float3 N = SurfaceNormal;\r\n\tconst float3 T = normalize(SurfaceTangent);\r\n\tconst float3 B = normalize(cross(SurfaceTangent, SurfaceNormal));\r\n\tconst float3 Out = cos(AngleRad) * T + sin(AngleRad) * B;\r\n\treturn Out;\r\n}"
   description=RotateTangentVec
   inputs [
    {inputName=Rotation, input {expression=&self:MaterialExpressionFrac_0}}
    {
     inputName=SurfaceNormal
     input {expression=&self:MaterialExpressionNamedRerouteUsage_76}
    }
    {
     inputName=SurfaceTangent
     input {expression=&self:MaterialExpressionNamedRerouteUsage_0}
    }
   ]
  }
  NamedRerouteDeclaration(None=@n62) {name=SpecAnisoRot, nodeColor=<rgba 1 0 0.0469 1>}
 ]
}
