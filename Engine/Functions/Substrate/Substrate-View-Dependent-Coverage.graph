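# Substrate helper material function that makes a layer's coverage view-dependent, based on its thickness.
# It computes saturated N·V from the camera vector and a Normal input that defaults to the world-space pixel normal.
# A Custom node derives a Smith-Schlick G term from N·V and thickness, then lerps coverage toward 1 by it, following "Practical Multilayered Materials" (Siggraph 2017).
# It outputs the modified coverage and the thickness, clamped to non-negative.
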MaterialFunction /Engine/Functions/Substrate/Substrate-View-Dependent-Coverage
graph {
 properties {bExposeToLibrary=true, userExposedCaption=Substrate View-Dependent-Coverage, libraryCategoriesText [Substrate Helpers]}
 nodes [
  MaterialExpressionMax_0 = Max(A=FunctionInput(Preview=Constant() {r=0.01}) {inputName=Thickness, inputType=FunctionInput_Scalar}) {constB=0}
  FunctionOutput(None=@MaterialExpressionMax_0) {description=The resulting thickness. Should be unchanged compare to the input., outputName=Thickness}
  MaterialExpressionSaturate_1 = Saturate(None=FunctionInput(Preview=Constant()) {inputName=Coverage, inputType=FunctionInput_Scalar})
  n1 = PixelNormalWS()
  MaterialExpressionSaturate_0 = Saturate(None=DotProduct(A=CameraVectorWS(), B=FunctionInput(Preview=@n1) {inputName=Normal}))
  n2 = Custom!(Thickness=@MaterialExpressionMax_0, Coverage=@MaterialExpressionSaturate_1, NoV=@MaterialExpressionSaturate_0) {
   code="\t// Reference \"Practical Multilayered Materials\" Siggraph 2017\r\n\tconst float x = NoV;\r\n\tconst float t = Thickness;\t\r\n\tconst float alpha = Coverage;\t\r\n\tconst float g = 1.0f - ( x / ( x * ( 1.0 - t ) + t ) ); // Smith-Schlick G\r\n\treturn lerp( alpha, 1.0f, g );"
   outputType=CMOT_Float1
   inputs [
    {inputName=Thickness, input {expression=&self:MaterialExpressionMax_0}}
    {inputName=Coverage, input {expression=&self:MaterialExpressionSaturate_1}}
    {inputName=NoV, input {expression=&self:MaterialExpressionSaturate_0}}
   ]
  }
  FunctionOutput(None=@n2) {description=The resulting modified coverage., outputName=Coverage}
 ]
}
